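# Translucent unlit text material in EntityFramework that renders distance-field font text turned toward the camera.
# AlignMeshToTheCamera supplies the billboard offset using custom basis vectors, and a vertex color node provides per-text tinting.
# A Font parameter defaulting to Roboto distance field is thresholded at its midpoint with Subtract and Sign, giving a hard-edged glyph mask scaled by an Opacity parameter.
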Material /EntityFramework/Materials/TextMaterial
graph {
 properties {blendMode=BLEND_Translucent, pixelDepthOffsetMode=PDOM_Legacy, shadingModel=MSM_Unlit, usedShadingModels=Unlit}
 nodes [
  MaterialFunctionCall!(
   WorldPosition
   PivotLocation
   WS Normals
   Custom Object Basis 1
   Custom Object Basis 2=Constant3Vector() {constant=<rgba 1 0 0 0>}
   Custom Object Basis 3=Constant3Vector() {constant=<rgba 0 -1 0 0>}
  ) {materialFunction=/Engine/Functions/Engine_MaterialFunctions02/WorldPositionOffset/AlignMeshToTheCamera}
  n1 = FontSampleParameter() {font=/Engine/EngineFonts/RobotoDistanceField, parameterName=Font}
  Multiply(A=Sign(None=Subtract(A=@n1.R) {constB=0.5}), B=ScalarParameter() {defaultValue=1, parameterName=Opacity})
  VertexColor()
 ]
}
